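# Sound cue for the Storm King hammer's fire-element impact, named for hits on husk enemies.
# A Mixer layers two equally weighted Random picks: one of six short enemy fire impact sounds and one of five default hammer impact sounds.
# The mix feeds a root Modulator and plays with small melee weapon attenuation under the weapon melee sound class.
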SoundCue /Game/Sounds/Fort_Weapons/Storm_King/StormKing_Hammer/SK_Hammer_Impact_Fire_Husk_Cue
root=Modulator_2
nodes [
 {id=Mixer_0, children [Random_0,Random_1], properties {inputVolume [1,1]}}
 {id=Modulator_2, children [Mixer_0]}
 {id=Random_0, children [WavePlayer_1,WavePlayer_2,WavePlayer_3,WavePlayer_4,WavePlayer_6,WavePlayer_7], properties {weights [1,1,1,1,1,1]}}
 {id=Random_1, children [WavePlayer_0, WavePlayer_5, WavePlayer_8, WavePlayer_9, WavePlayer_10], properties {weights [1,1,1,1,1]}}
 {id=WavePlayer_0, properties {soundWaveAssetPtr=&/StormKing_Hammer_Default_Impact_02}}
 {id=WavePlayer_1, properties {soundWaveAssetPtr=/Game/Sounds/Fort_GamePlay_Sounds/Elemental_Layers/enemy_fire_short_impact_01}}
 {id=WavePlayer_10, properties {soundWaveAssetPtr=&/StormKing_Hammer_Default_Impact_05}}
 {id=WavePlayer_2, properties {soundWaveAssetPtr=/Game/Sounds/Fort_GamePlay_Sounds/Elemental_Layers/enemy_fire_short_impact_02}}
 {id=WavePlayer_3, properties {soundWaveAssetPtr=/Game/Sounds/Fort_GamePlay_Sounds/Elemental_Layers/enemy_fire_short_impact_03}}
 {id=WavePlayer_4, properties {soundWaveAssetPtr=/Game/Sounds/Fort_GamePlay_Sounds/Elemental_Layers/enemy_fire_short_impact_04}}
 {id=WavePlayer_5, properties {soundWaveAssetPtr=&/StormKing_Hammer_Default_Impact_01}}
 {id=WavePlayer_6, properties {soundWaveAssetPtr=/Game/Sounds/Fort_GamePlay_Sounds/Elemental_Layers/enemy_fire_short_impact_06}}
 {id=WavePlayer_7, properties {soundWaveAssetPtr=/Game/Sounds/Fort_GamePlay_Sounds/Elemental_Layers/enemy_fire_short_impact_07}}
 {id=WavePlayer_8, properties {soundWaveAssetPtr=&/StormKing_Hammer_Default_Impact_03}}
 {id=WavePlayer_9, properties {soundWaveAssetPtr=&/StormKing_Hammer_Default_Impact_04}}
]
properties {
 attenuationSettings=&Fort_Weapon_Melee_Sml_Att
 maxDistance=2800
 soundClassObject=&Fort_WeaponMelee_PSM
 volumeMultiplier=1
}
