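# Sound cue for Storm King hammer impacts on stone, classed as impact audio under player melee weapon concurrency.
# Three attenuated layers crossfade by distance: close hammer hits plus delayed stone, delayed pitch-varied medium rock, and delayed distant rock.
# Each layer picks randomly from wave variations.
# A quality switch provides a full first branch and a leaner second with fewer variations and no Storm King-specific samples, leaving two slots empty.
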SoundCue /Game/Sounds/Fort_Weapons/Storm_King/StormKing_Hammer/SK_Hammer_Impact_Stone_MLTI_Cue
root=QualityLevel_0
nodes [
 {
  id=Attenuation_3
  children [Mixer_5]
  properties {
   attenuationSettings=/Game/Sounds/Attenuation/Fort_Demolition_Crossfade_Global_Hit_Close_Att
   attenuationOverrides {falloffDistance=900, attenuationShapeExtents {x=100}}
  }
 }
 {
  id=Attenuation_4
  children [Delay_4]
  properties {attenuationSettings=/Game/Sounds/Attenuation/Fort_Demolition_Crossfade_Global_Hit_Medium_Att, attenuationOverrides {falloffDistance=1600}}
 }
 {
  id=Attenuation_5
  children [Modulator_5]
  properties {
   attenuationSettings=/Game/Sounds/Attenuation/Fort_Demolition_Crossfade_Global_Hit_Far_Att
   attenuationOverrides {falloffDistance=3000, attenuationShapeExtents {x=1000}}
  }
 }
 {
  id=Attenuation_6
  children [Mixer_8]
  properties {
   attenuationSettings=&Fort_Demolition_Crossfade_Global_Hit_Close_Att
   attenuationOverrides {falloffDistance=900, attenuationShapeExtents {x=100}}
  }
 }
 {
  id=Attenuation_7
  children [Delay_7]
  properties {attenuationSettings=&Fort_Demolition_Crossfade_Global_Hit_Medium_Att, attenuationOverrides {falloffDistance=1600}}
 }
 {
  id=Attenuation_8
  children [Modulator_9]
  properties {
   attenuationSettings=&Fort_Demolition_Crossfade_Global_Hit_Far_Att
   attenuationOverrides {falloffDistance=3000, attenuationShapeExtents {x=1000}}
  }
 }
 {id=Delay_3, children [Modulator_6], properties {delayMax=0.1, delayMin=0.1}}
 {id=Delay_4, children [Modulator_4], properties {delayMax=0.1, delayMin=0.1}}
 {id=Delay_5, children [Random_4], properties {delayMax=0.1, delayMin=0.1}}
 {id=Delay_6, children [Modulator_10], properties {delayMax=0.1, delayMin=0.1}}
 {id=Delay_7, children [Modulator_8], properties {delayMax=0.1, delayMin=0.1}}
 {id=Delay_8, children [Random_8], properties {delayMax=0.1, delayMin=0.1}}
 {id=Mixer_0, children [Random_7,Random_0], properties {inputVolume [1,1]}}
 {id=Mixer_3, children [Attenuation_3, Attenuation_4], properties {inputVolume [1,1]}}
 {id=Mixer_4, children [Mixer_3, Attenuation_5], properties {inputVolume [1,1]}}
 {id=Mixer_5, children [Modulator_7,Delay_3], properties {inputVolume [1.6,1]}}
 {id=Mixer_6, children [Attenuation_6, Attenuation_7], properties {inputVolume [1,1]}}
 {id=Mixer_7, children [Mixer_6, Attenuation_8], properties {inputVolume [1,1]}}
 {id=Mixer_8, children [Modulator_11,Delay_6], properties {inputVolume [1.6,1]}}
 {id=Modulator_10, children [Random_10]}
 {id=Modulator_11, children [Random_11]}
 {id=Modulator_4, children [Random_5], properties {pitchMax=1.2, pitchMin=0.8, volumeMax=2, volumeMin=2}}
 {id=Modulator_5, children [Delay_5], properties {pitchMax=1.4}}
 {id=Modulator_6, children [Random_6]}
 {id=Modulator_7, children [Mixer_0]}
 {id=Modulator_8, children [Random_9], properties {pitchMax=1.2, pitchMin=0.8, volumeMax=2, volumeMin=2}}
 {id=Modulator_9, children [Delay_8], properties {pitchMax=1.4}}
 {id=QualityLevel_0, children [Mixer_4,Mixer_7,null,null]}
 {id=Random_0, children [WavePlayer_0,WavePlayer_1,WavePlayer_2,WavePlayer_3,WavePlayer_4], properties {weights [1,1,1,1,1]}}
 {
  id=Random_10
  children [WavePlayer_52, WavePlayer_53, WavePlayer_56, WavePlayer_58, WavePlayer_60, WavePlayer_61]
  properties {weights [1,1,1,1,1,1]}
 }
 {id=Random_11, children [WavePlayer_62, WavePlayer_63, WavePlayer_64], properties {weights [1,1,1]}}
 {id=Random_4, children [WavePlayer_23, WavePlayer_22, WavePlayer_24, WavePlayer_25], properties {weights [1,1,1,1]}}
 {id=Random_5, children [WavePlayer_26, WavePlayer_27, WavePlayer_28, WavePlayer_29], properties {weights [1,1,1,1]}}
 {
  id=Random_6
  children [
   WavePlayer_30
   WavePlayer_31
   WavePlayer_32
   WavePlayer_33
   WavePlayer_34
   WavePlayer_35
   WavePlayer_36
   WavePlayer_37
   WavePlayer_38
   WavePlayer_39
  ]
  properties {weights [1,1,1,1,1,1,1,1,1,1]}
 }
 {id=Random_7, children [WavePlayer_40, WavePlayer_41, WavePlayer_42, WavePlayer_43], properties {weights [1,1,1,1]}}
 {id=Random_8, children [WavePlayer_44, WavePlayer_46, WavePlayer_47], properties {weights [1,1,1]}}
 {id=Random_9, children [WavePlayer_48, WavePlayer_49, WavePlayer_50], properties {weights [1,1,1]}}
 {id=WavePlayer_0, properties {soundWaveAssetPtr=&/StormKing_Hammer_Default_Impact_02}}
 {id=WavePlayer_1, properties {soundWaveAssetPtr=&/StormKing_Hammer_Default_Impact_01}}
 {id=WavePlayer_2, properties {soundWaveAssetPtr=&/StormKing_Hammer_Default_Impact_03}}
 {id=WavePlayer_22, properties {soundWaveAssetPtr=/Game/Sounds/Fort_Impact_Sounds/Demolition/Defaults/Rock_Distant_Impact_10}}
 {id=WavePlayer_23, properties {soundWaveAssetPtr=/Game/Sounds/Fort_Impact_Sounds/Demolition/Defaults/Rock_Distant_Impact_02}}
 {id=WavePlayer_24, properties {soundWaveAssetPtr=/Game/Sounds/Fort_Impact_Sounds/Demolition/Defaults/Rock_Distant_Impact_03}}
 {id=WavePlayer_25, properties {soundWaveAssetPtr=/Game/Sounds/Fort_Impact_Sounds/Demolition/Defaults/Rock_Distant_Impact_07}}
 {id=WavePlayer_26, properties {soundWaveAssetPtr=/Game/Sounds/Fort_Impact_Sounds/Demolition/Defaults/rock_newimp_03}}
 {id=WavePlayer_27, properties {soundWaveAssetPtr=/Game/Sounds/Fort_Impact_Sounds/Demolition/Defaults/rock_newimp_11}}
 {id=WavePlayer_28, properties {soundWaveAssetPtr=/Game/Sounds/Fort_Impact_Sounds/Demolition/Defaults/rock_newimp_12}}
 {id=WavePlayer_29, properties {soundWaveAssetPtr=/Game/Sounds/Fort_Impact_Sounds/Demolition/Defaults/rock_newimp_01}}
 {id=WavePlayer_3, properties {soundWaveAssetPtr=&/StormKing_Hammer_Default_Impact_04}}
 {id=WavePlayer_30, properties {soundWaveAssetPtr=/Game/Sounds/Fort_Impact_Sounds/Demolition/Defaults/default_stone_impact_new_combined_10}}
 {id=WavePlayer_31, properties {soundWaveAssetPtr=/Game/Sounds/Fort_Impact_Sounds/Demolition/Defaults/default_stone_impact_new_combined_01}}
 {id=WavePlayer_32, properties {soundWaveAssetPtr=/Game/Sounds/Fort_Impact_Sounds/Demolition/Defaults/default_stone_impact_new_combined_02}}
 {id=WavePlayer_33, properties {soundWaveAssetPtr=/Game/Sounds/Fort_Impact_Sounds/Demolition/Defaults/default_stone_impact_new_combined_03}}
 {id=WavePlayer_34, properties {soundWaveAssetPtr=/Game/Sounds/Fort_Impact_Sounds/Demolition/Defaults/default_stone_impact_new_combined_04}}
 {id=WavePlayer_35, properties {soundWaveAssetPtr=/Game/Sounds/Fort_Impact_Sounds/Demolition/Defaults/default_stone_impact_new_combined_05}}
 {id=WavePlayer_36, properties {soundWaveAssetPtr=/Game/Sounds/Fort_Impact_Sounds/Demolition/Defaults/default_stone_impact_new_combined_06}}
 {id=WavePlayer_37, properties {soundWaveAssetPtr=/Game/Sounds/Fort_Impact_Sounds/Demolition/Defaults/default_stone_impact_new_combined_07}}
 {id=WavePlayer_38, properties {soundWaveAssetPtr=/Game/Sounds/Fort_Impact_Sounds/Demolition/Defaults/default_stone_impact_new_combined_08}}
 {id=WavePlayer_39, properties {soundWaveAssetPtr=/Game/Sounds/Fort_Impact_Sounds/Demolition/Defaults/default_stone_impact_new_combined_09}}
 {id=WavePlayer_4, properties {soundWaveAssetPtr=&/StormKing_Hammer_Default_Impact_05}}
 {id=WavePlayer_40, properties {soundWaveAssetPtr=/Game/Sounds/Fort_Weapons/Sledgehammer/sledgehammer_impact_01}}
 {id=WavePlayer_41, properties {soundWaveAssetPtr=/Game/Sounds/Fort_Weapons/Sledgehammer/sledgehammer_impact_02}}
 {id=WavePlayer_42, properties {soundWaveAssetPtr=/Game/Sounds/Fort_Weapons/Sledgehammer/sledgehammer_impact_04}}
 {id=WavePlayer_43, properties {soundWaveAssetPtr=/Game/Sounds/Fort_Weapons/Sledgehammer/sledgehammer_impact_05}}
 {id=WavePlayer_44, properties {soundWaveAssetPtr=&Rock_Distant_Impact_10}}
 {id=WavePlayer_46, properties {soundWaveAssetPtr=&Rock_Distant_Impact_03}}
 {id=WavePlayer_47, properties {soundWaveAssetPtr=&Rock_Distant_Impact_07}}
 {id=WavePlayer_48, properties {soundWaveAssetPtr=&rock_newimp_03}}
 {id=WavePlayer_49, properties {soundWaveAssetPtr=&rock_newimp_11}}
 {id=WavePlayer_50, properties {soundWaveAssetPtr=&rock_newimp_12}}
 {id=WavePlayer_52, properties {soundWaveAssetPtr=&default_stone_impact_new_combined_10}}
 {id=WavePlayer_53, properties {soundWaveAssetPtr=&default_stone_impact_new_combined_01}}
 {id=WavePlayer_56, properties {soundWaveAssetPtr=&default_stone_impact_new_combined_04}}
 {id=WavePlayer_58, properties {soundWaveAssetPtr=&default_stone_impact_new_combined_06}}
 {id=WavePlayer_60, properties {soundWaveAssetPtr=&default_stone_impact_new_combined_08}}
 {id=WavePlayer_61, properties {soundWaveAssetPtr=&default_stone_impact_new_combined_09}}
 {id=WavePlayer_62, properties {soundWaveAssetPtr=&sledgehammer_impact_01}}
 {id=WavePlayer_63, properties {soundWaveAssetPtr=&sledgehammer_impact_02}}
 {id=WavePlayer_64, properties {soundWaveAssetPtr=&sledgehammer_impact_04}}
]
properties {
 maxDistance=8000
 soundClassObject=&Fort_Impacts
 volumeMultiplier=0.442
 concurrencyOverrides {resolutionRule=PreventNew}
 concurrencySet [&Player_Melee_Weapon_Conc]
}
