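# Sound cue for the Storm King hammer striking wood, routed to the Fort_Impacts sound class.
# Three layers, each with close, medium or far crossfade attenuation, are mixed to vary the hit by distance.
# Close blends randomized Storm King and sledgehammer impacts with a slightly delayed wood hit.
# Medium and far play delayed, pitch-modulated random wood and distant wood impacts.
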SoundCue /Game/Sounds/Fort_Weapons/Storm_King/StormKing_Hammer/SK_Hammer_Impact_Wood_MLTI_Cue
root=Mixer_4
nodes [
 {
  id=Attenuation_3
  children [Mixer_5]
  properties {
   attenuationSettings=/Game/Sounds/Attenuation/Fort_Demolition_Crossfade_Global_Hit_Close_Att
   attenuationOverrides {falloffDistance=900, attenuationShapeExtents {x=100}}
  }
 }
 {
  id=Attenuation_4
  children [Delay_4]
  properties {attenuationSettings=/Game/Sounds/Attenuation/Fort_Demolition_Crossfade_Global_Hit_Medium_Att, attenuationOverrides {falloffDistance=1600}}
 }
 {
  id=Attenuation_5
  children [Delay_5]
  properties {
   attenuationSettings=/Game/Sounds/Attenuation/Fort_Demolition_Crossfade_Global_Hit_Far_Att
   attenuationOverrides {falloffDistance=3000, attenuationShapeExtents {x=1000}}
  }
 }
 {id=Delay_3, children [Random_6], properties {delayMax=0.1, delayMin=0.1}}
 {id=Delay_4, children [Modulator_3], properties {delayMax=0.2, delayMin=0.1}}
 {id=Delay_5, children [Modulator_4], properties {delayMax=0.2, delayMin=0.1}}
 {id=Mixer_0, children [Random_7,Random_0], properties {inputVolume [1,1]}}
 {id=Mixer_3, children [Attenuation_3, Attenuation_4], properties {inputVolume [1,0.5]}}
 {id=Mixer_4, children [Mixer_3, Attenuation_5], properties {inputVolume [1.2,0.85]}}
 {id=Mixer_5, children [Modulator_5,Delay_3], properties {inputVolume [1.6,0.5]}}
 {id=Modulator_3, children [Random_4], properties {pitchMax=1.6, pitchMin=1.5}}
 {id=Modulator_4, children [Random_5], properties {pitchMax=1.5}}
 {id=Modulator_5, children [Mixer_0]}
 {id=Random_0, children [WavePlayer_0,WavePlayer_1,WavePlayer_2,WavePlayer_3,WavePlayer_4], properties {weights [1,1,1,1,1]}}
 {id=Random_4, children [WavePlayer_23, WavePlayer_22, WavePlayer_21, WavePlayer_20], properties {weights [1,1,1,1]}}
 {id=Random_5, children [WavePlayer_24, WavePlayer_25, WavePlayer_26, WavePlayer_27], properties {weights [1,1,1,1]}}
 {
  id=Random_6
  children [WavePlayer_28, WavePlayer_29, WavePlayer_30, WavePlayer_31, WavePlayer_32, WavePlayer_33, WavePlayer_34, WavePlayer_35]
  properties {weights [1,1,1,1,1,1,1,1]}
 }
 {id=Random_7, children [WavePlayer_36, WavePlayer_37, WavePlayer_38, WavePlayer_39], properties {weights [1,1,1,1]}}
 {id=WavePlayer_0, properties {soundWaveAssetPtr=&/StormKing_Hammer_Default_Impact_02}}
 {id=WavePlayer_1, properties {soundWaveAssetPtr=&/StormKing_Hammer_Default_Impact_01}}
 {id=WavePlayer_2, properties {soundWaveAssetPtr=&/StormKing_Hammer_Default_Impact_03}}
 {id=WavePlayer_20, properties {soundWaveAssetPtr=/Game/Sounds/Fort_Impact_Sounds/Demolition/Defaults/wood_impact_09}}
 {id=WavePlayer_21, properties {soundWaveAssetPtr=/Game/Sounds/Fort_Impact_Sounds/Demolition/Defaults/wood_impact_07}}
 {id=WavePlayer_22, properties {soundWaveAssetPtr=/Game/Sounds/Fort_Impact_Sounds/Demolition/Defaults/wood_impact_02}}
 {id=WavePlayer_23, properties {soundWaveAssetPtr=/Game/Sounds/Fort_Impact_Sounds/Demolition/Defaults/wood_impact_01}}
 {id=WavePlayer_24, properties {soundWaveAssetPtr=/Game/Sounds/Fort_Impact_Sounds/Demolition/Defaults/wood_distant_impact_02}}
 {id=WavePlayer_25, properties {soundWaveAssetPtr=/Game/Sounds/Fort_Impact_Sounds/Demolition/Defaults/wood_distant_impact_04}}
 {id=WavePlayer_26, properties {soundWaveAssetPtr=/Game/Sounds/Fort_Impact_Sounds/Demolition/Defaults/wood_distant_impact_05}}
 {id=WavePlayer_27, properties {soundWaveAssetPtr=/Game/Sounds/Fort_Impact_Sounds/Demolition/Defaults/wood_distant_impact_06}}
 {id=WavePlayer_28, properties {soundWaveAssetPtr=/Game/Sounds/Fort_Impact_Sounds/Demolition/Defaults/default_wood_impact_new_combined_01}}
 {id=WavePlayer_29, properties {soundWaveAssetPtr=/Game/Sounds/Fort_Impact_Sounds/Demolition/Defaults/default_wood_impact_new_combined_02}}
 {id=WavePlayer_3, properties {soundWaveAssetPtr=&/StormKing_Hammer_Default_Impact_04}}
 {id=WavePlayer_30, properties {soundWaveAssetPtr=/Game/Sounds/Fort_Impact_Sounds/Demolition/Defaults/default_wood_impact_new_combined_03}}
 {id=WavePlayer_31, properties {soundWaveAssetPtr=/Game/Sounds/Fort_Impact_Sounds/Demolition/Defaults/default_wood_impact_new_combined_04}}
 {id=WavePlayer_32, properties {soundWaveAssetPtr=/Game/Sounds/Fort_Impact_Sounds/Demolition/Defaults/default_wood_impact_new_combined_05}}
 {id=WavePlayer_33, properties {soundWaveAssetPtr=/Game/Sounds/Fort_Impact_Sounds/Demolition/Defaults/default_wood_impact_new_combined_06}}
 {id=WavePlayer_34, properties {soundWaveAssetPtr=/Game/Sounds/Fort_Impact_Sounds/Demolition/Defaults/default_wood_impact_new_combined_07}}
 {id=WavePlayer_35, properties {soundWaveAssetPtr=/Game/Sounds/Fort_Impact_Sounds/Demolition/Defaults/default_wood_impact_new_combined_08}}
 {id=WavePlayer_36, properties {soundWaveAssetPtr=/Game/Sounds/Fort_Weapons/Sledgehammer/sledgehammer_impact_01}}
 {id=WavePlayer_37, properties {soundWaveAssetPtr=/Game/Sounds/Fort_Weapons/Sledgehammer/sledgehammer_impact_02}}
 {id=WavePlayer_38, properties {soundWaveAssetPtr=/Game/Sounds/Fort_Weapons/Sledgehammer/sledgehammer_impact_04}}
 {id=WavePlayer_39, properties {soundWaveAssetPtr=/Game/Sounds/Fort_Weapons/Sledgehammer/sledgehammer_impact_05}}
 {id=WavePlayer_4, properties {soundWaveAssetPtr=&/StormKing_Hammer_Default_Impact_05}}
]
properties {
 maxDistance=8000
 soundClassObject=&Fort_Impacts
 volumeMultiplier=0.442
 concurrencyOverrides {maxCount=1, resolutionRule=StopOldest}
 concurrencySet [&Player_Melee_Weapon_Conc]
}
