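# Impact sound cue for the Storm King hammer's heavy attack, mixing a close layer and a distant layer.
# Each layer randomly picks a wave, passes it through a modulator, then applies its own attenuation and sound class override.
# The quieter distant layer carries farther with low-pass filtering over distance and borrows the ranged-weapon distant-fire class, while the close layer uses a melee class.
# Player melee weapon concurrency governs the whole cue.
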SoundCue /Game/Sounds/Fort_Weapons/Storm_King/StormKing_Hammer/StormKing_Hammer_Heavy_Attack_Impact_Cue
root=Mixer_1
nodes [
 {
  id=Attenuation_2
  children [Modulator_0]
  properties {
   attenuationSettings=&Fort_Weapon_Melee_Sml_Att
   bOverrideAttenuation=true
   attenuationOverrides {falloffDistance=2000, stereoSpread=2000, attenuationShapeExtents {x=800}}
  }
 }
 {
  id=Attenuation_3
  children [Modulator_1]
  properties {
   bOverrideAttenuation=true
   attenuationOverrides {bAttenuateWithLPF=true, falloffDistance=2000, lPFRadiusMax=5000, lPFRadiusMin=2000, attenuationShapeExtents {x=3000}}
  }
 }
 {id=Mixer_1, children [SoundClass_3,SoundClass_2], properties {inputVolume [1,0.8]}}
 {id=Modulator_0, children [Random_0]}
 {id=Modulator_1, children [Random_2]}
 {id=Random_0, children [WavePlayer_0,WavePlayer_1,WavePlayer_2], properties {weights [1,1,1]}}
 {id=Random_2, children [WavePlayer_3, WavePlayer_4, WavePlayer_5, WavePlayer_6, WavePlayer_10], properties {weights [1,1,1,1,1]}}
 {
  id=SoundClass_2
  children [Attenuation_3]
  properties {soundClassOverride=&Fort_WeaponRanged_Fire_Distant}
 }
 {id=SoundClass_3, children [Attenuation_2], properties {soundClassOverride=&Fort_WeaponMelee_PSM}}
 {id=WavePlayer_0, properties {soundWaveAssetPtr=&/StormKing_Hammer_Heavy_Attack_Close_01}}
 {id=WavePlayer_1, properties {soundWaveAssetPtr=&/StormKing_Hammer_Heavy_Attack_Close_02}}
 {
  id=WavePlayer_10
  properties {soundWaveAssetPtr=&/StormKing_Hammer_Heavy_Attack_Distant_05}
 }
 {id=WavePlayer_2, properties {soundWaveAssetPtr=&/StormKing_Hammer_Heavy_Attack_Close_03}}
 {
  id=WavePlayer_3
  properties {soundWaveAssetPtr=&/StormKing_Hammer_Heavy_Attack_Distant_01}
 }
 {
  id=WavePlayer_4
  properties {soundWaveAssetPtr=&/StormKing_Hammer_Heavy_Attack_Distant_02}
 }
 {
  id=WavePlayer_5
  properties {soundWaveAssetPtr=&/StormKing_Hammer_Heavy_Attack_Distant_03}
 }
 {
  id=WavePlayer_6
  properties {soundWaveAssetPtr=&/StormKing_Hammer_Heavy_Attack_Distant_04}
 }
]
properties {
 maxDistance=5000
 soundClassObject=&Fort_WeaponMelee_PSM
 volumeMultiplier=1.5
 attenuationOverrides {distanceAlgorithm=NaturalSound, falloffDistance=3000, attenuationShapeExtents {x=1000}}
 concurrencySet [&Player_Melee_Weapon_Conc]
}
