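# Sound cue for the heavy attack swing of the Storm King hammer melee weapon.
# A mixer layers two modulated branches, each picking one of four equally weighted waves at random: an overhead swing and a heavy-swing "Swt" layer.
# It uses small melee weapon attenuation, the melee weapon sound class and player melee concurrency, with a boosted overall volume.
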SoundCue /Game/Sounds/Fort_Weapons/Storm_King/StormKing_Hammer/StormKing_Hammer_Heavy_Attack_Swing_Cue
root=Mixer_0
nodes [
 {id=Mixer_0, children [Modulator_0,Modulator_1], properties {inputVolume [1,1]}}
 {id=Modulator_0, children [Random_1]}
 {id=Modulator_1, children [Random_0]}
 {id=Random_0, children [WavePlayer_0,WavePlayer_1,WavePlayer_2,WavePlayer_3], properties {weights [1,1,1,1]}}
 {id=Random_1, children [WavePlayer_7, WavePlayer_8, WavePlayer_9, WavePlayer_10], properties {weights [1,1,1,1]}}
 {id=WavePlayer_0, properties {soundWaveAssetPtr=&/StormKing_Hammer_Heavy_Swing_Swt_04}}
 {id=WavePlayer_1, properties {soundWaveAssetPtr=&/StormKing_Hammer_Heavy_Swing_Swt_01}}
 {id=WavePlayer_10, properties {soundWaveAssetPtr=&/StormKing_Hammer_SwingOverhead_04}}
 {id=WavePlayer_2, properties {soundWaveAssetPtr=&/StormKing_Hammer_Heavy_Swing_Swt_02}}
 {id=WavePlayer_3, properties {soundWaveAssetPtr=&/StormKing_Hammer_Heavy_Swing_Swt_03}}
 {id=WavePlayer_7, properties {soundWaveAssetPtr=&/StormKing_Hammer_SwingOverhead_01}}
 {id=WavePlayer_8, properties {soundWaveAssetPtr=&/StormKing_Hammer_SwingOverhead_02}}
 {id=WavePlayer_9, properties {soundWaveAssetPtr=&/StormKing_Hammer_SwingOverhead_03}}
]
properties {
 attenuationSettings=&Fort_Weapon_Melee_Sml_Att
 maxDistance=2800
 soundClassObject=&Fort_WeaponMelee_PSM
 volumeMultiplier=1.3
 concurrencySet [&Player_Melee_Weapon_Conc]
}
