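# Sound cue for the Storm King hammer melee weapon's impact sound.
# The first quality level mixes a random default impact with a quieter, slightly delayed random hammer impact layer.
# The other three quality levels share a simpler random pick of three impact waves, avoiding the layered mix.
# Playback uses pickaxe impact attenuation and player melee weapon concurrency.
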SoundCue /Game/Sounds/Fort_Weapons/Storm_King/StormKing_Hammer/StormKing_Hammer_Impact_Cue
root=QualityLevel_0
nodes [
 {id=Delay_0, children [Random_3], properties {delayMin=0.03}}
 {id=Mixer_0, children [Random_0,Delay_0], properties {inputVolume [1,0.6]}}
 {id=QualityLevel_0, children [Mixer_0,Random_4,Random_4,Random_4]}
 {id=Random_0, children [WavePlayer_0,WavePlayer_1,WavePlayer_2,WavePlayer_3,WavePlayer_4], properties {weights [1,1,1,1,1]}}
 {
  id=Random_3
  children [WavePlayer_12, WavePlayer_13, WavePlayer_14, WavePlayer_17, WavePlayer_18, WavePlayer_19]
  properties {weights [1,1,1,1,1,1]}
 }
 {id=Random_4, children [WavePlayer_20, WavePlayer_21, WavePlayer_22], properties {weights [1,1,1]}}
 {id=WavePlayer_0, properties {soundWaveAssetPtr=&/StormKing_Hammer_Default_Impact_01}}
 {id=WavePlayer_1, properties {soundWaveAssetPtr=&/StormKing_Hammer_Default_Impact_02}}
 {id=WavePlayer_12, properties {soundWaveAssetPtr=&/StormKing_Hammer_Impact_01}}
 {id=WavePlayer_13, properties {soundWaveAssetPtr=&/StormKing_Hammer_Impact_06}}
 {id=WavePlayer_14, properties {soundWaveAssetPtr=&/StormKing_Hammer_Impact_02}}
 {id=WavePlayer_17, properties {soundWaveAssetPtr=&/StormKing_Hammer_Impact_03}}
 {id=WavePlayer_18, properties {soundWaveAssetPtr=&/StormKing_Hammer_Impact_04}}
 {id=WavePlayer_19, properties {soundWaveAssetPtr=&/StormKing_Hammer_Impact_05}}
 {id=WavePlayer_2, properties {soundWaveAssetPtr=&/StormKing_Hammer_Default_Impact_03}}
 {id=WavePlayer_20, properties {soundWaveAssetPtr=&StormKing_Hammer_Impact_01}}
 {id=WavePlayer_21, properties {soundWaveAssetPtr=&StormKing_Hammer_Impact_02}}
 {id=WavePlayer_22, properties {soundWaveAssetPtr=&StormKing_Hammer_Impact_03}}
 {id=WavePlayer_3, properties {soundWaveAssetPtr=&/StormKing_Hammer_Default_Impact_04}}
 {id=WavePlayer_4, properties {soundWaveAssetPtr=&/StormKing_Hammer_Default_Impact_05}}
]
properties {
 attenuationSettings=/Game/Sounds/Attenuation/Fort_Weapon_Pickaxe_Player_Impact_Att
 maxDistance=3512
 soundClassObject=&Fort_WeaponMelee_PSM
 volumeMultiplier=1
 attenuationOverrides {distanceAlgorithm=NaturalSound, falloffDistance=3000, attenuationShapeExtents {x=1000}}
 concurrencyOverrides {resolutionRule=PreventNew}
 concurrencySet [&Player_Melee_Weapon_Conc]
}
