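# Sound cue for readying the Storm King hammer melee weapon, built from reused swing sounds.
# A mixer blends a quiet layer of random Hydraulic hammer swings with a full-volume layer of random Storm King hammer swings, each passed through a modulator.
# The Storm King layer is randomly pitched well up and shaped by an envelope that fades it out.
# The whole cue is pitched up and uses small-melee attenuation.
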SoundCue /Game/Sounds/Fort_Weapons/Storm_King/StormKing_Hammer/StormKing_Hammer_Ready_Cue
root=Mixer_0
nodes [
 {
  id=Enveloper_0
  children [Random_1]
  properties {
   pitchCurve {editorCurveData {keys [{time=0}]}}
   volumeCurve {editorCurveData {keys [<at 0 1 RCIM_Cubic>, {arriveTangent=-0.0475, interpMode=RCIM_Cubic, leaveTangent=-0.0475, tangentMode=RCTM_User, time=1.5, value=0}]}}
  }
 }
 {id=Mixer_0, children [Modulator_0,Modulator_1], properties {inputVolume [0.2,1]}}
 {id=Modulator_0, children [Random_0]}
 {id=Modulator_1, children [Enveloper_0], properties {pitchMax=3, pitchMin=1.5}}
 {id=Random_0, children [WavePlayer_1,WavePlayer_2,WavePlayer_3,WavePlayer_4], properties {weights [1,1,1,1]}}
 {id=Random_1, children [WavePlayer_0,WavePlayer_5,WavePlayer_6,WavePlayer_7,WavePlayer_8,WavePlayer_9], properties {weights [1,1,1,1,1,1]}}
 {id=WavePlayer_0, properties {soundWaveAssetPtr=&/StormKing_Hammer_Swing_04}}
 {id=WavePlayer_1, properties {soundWaveAssetPtr=/Game/Sounds/Fort_Weapons/Hydraulic/Hydraulic_Hammer_Swing_01}}
 {id=WavePlayer_2, properties {soundWaveAssetPtr=/Game/Sounds/Fort_Weapons/Hydraulic/Hydraulic_Hammer_Swing_02}}
 {id=WavePlayer_3, properties {soundWaveAssetPtr=/Game/Sounds/Fort_Weapons/Hydraulic/Hydraulic_Hammer_Swing_03}}
 {id=WavePlayer_4, properties {soundWaveAssetPtr=/Game/Sounds/Fort_Weapons/Hydraulic/Hydraulic_Hammer_Swing_04}}
 {id=WavePlayer_5, properties {soundWaveAssetPtr=&/StormKing_Hammer_Swing_01}}
 {id=WavePlayer_6, properties {soundWaveAssetPtr=&/StormKing_Hammer_Swing_02}}
 {id=WavePlayer_7, properties {soundWaveAssetPtr=&/StormKing_Hammer_Swing_03}}
 {id=WavePlayer_8, properties {soundWaveAssetPtr=&/StormKing_Hammer_Swing_05}}
 {id=WavePlayer_9, properties {soundWaveAssetPtr=&/StormKing_Hammer_Swing_06}}
]
properties {
 attenuationSettings=&Fort_Weapon_Melee_Sml_Att
 maxDistance=2800
 pitchMultiplier=1.4
 soundClassObject=&Fort_WeaponMelee
 volumeMultiplier=1
}
