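# Sound cue for the swing sound of the Storm King hammer melee weapon.
# A random node picks one of six equally weighted swing waves, then a modulator slightly varies pitch on each play.
# Playback uses the melee weapon sound class, small melee attenuation and player melee weapon concurrency, with a boosted volume multiplier.
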SoundCue /Game/Sounds/Fort_Weapons/Storm_King/StormKing_Hammer/StormKing_Hammer_Swing_Cue
root=Modulator_7
nodes [
 {id=Modulator_7, children [Random_0], properties {pitchMax=1.1, pitchMin=0.9}}
 {
  id=Random_0
  children [WavePlayer_0, WavePlayer_6, WavePlayer_7, WavePlayer_8, WavePlayer_9, WavePlayer_10]
  properties {weights [1,1,1,1,1,1]}
 }
 {id=WavePlayer_0, properties {soundWaveAssetPtr=&/StormKing_Hammer_Swing_01}}
 {id=WavePlayer_10, properties {soundWaveAssetPtr=&/StormKing_Hammer_Swing_06}}
 {id=WavePlayer_6, properties {soundWaveAssetPtr=&/StormKing_Hammer_Swing_02}}
 {id=WavePlayer_7, properties {soundWaveAssetPtr=&/StormKing_Hammer_Swing_03}}
 {id=WavePlayer_8, properties {soundWaveAssetPtr=&/StormKing_Hammer_Swing_04}}
 {id=WavePlayer_9, properties {soundWaveAssetPtr=&/StormKing_Hammer_Swing_05}}
]
properties {
 attenuationSettings=&Fort_Weapon_Melee_Sml_Att
 maxDistance=2800
 soundClassObject=&Fort_WeaponMelee_PSM
 volumeMultiplier=1.5
 attenuationOverrides {distanceAlgorithm=NaturalSound, falloffDistance=2000, attenuationShapeExtents {x=500}}
 concurrencySet [&Player_Melee_Weapon_Conc]
}
